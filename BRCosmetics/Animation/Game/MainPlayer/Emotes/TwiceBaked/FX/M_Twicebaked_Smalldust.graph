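# Unlit, masked Niagara sprite material for small dust in the TwiceBaked emote effects, built on a dirt clod cluster texture.
# The texture's alpha drives a slight bump offset for a second sample, whose RGB is tinted by particle color.
# A hard step on the un-offset alpha blacks out low-alpha areas before MF_FX_EmissiveAdjustments.
# Particle alpha minus inverted offset alpha forms an erosion-style mask as particle alpha falls.
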Material /BRCosmetics/Animation/Game/MainPlayer/Emotes/TwiceBaked/FX/M_Twicebaked_Smalldust
graph {
 properties {bUsedWithNiagaraSprites=true, blendMode=BLEND_Masked, pixelDepthOffsetMode=PDOM_Legacy, shadingModel=MSM_Unlit, usedShadingModels=Unlit}
 nodes [
  n1 = ParticleColor()
  n2 = TextureSample() {texture=/Game/Effects/Fort_Effects/Textures/Dirt/T-DirtClodCluster_D}
  n3 = TextureSample(UVs=BumpOffset(Height=@n2.A) {heightRatio=0.04}) {texture=&T-DirtClodCluster_D}
  n4 = Constant() {r=0.3}
  MaterialFunctionCall(Emissive=LinearInterpolate(A=Constant(), B=Multiply(A=@n1.RGB, B=@n3.RGB), Alpha=SmoothStep(Min=@n4, Max=@n4, Value=@n2.A))) {materialFunction=&MF_FX_EmissiveAdjustments}
  Subtract(A=@n1.A, B=OneMinus(None=@n3.A))
 ]
}
