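# Niagara effect for the TwiceBaked emote's crush moment, throwing rock debris, dust and smoke out to both sides.
# CrumbleDust ramps up over the system loop, while the other emitters fire after User.Delay.
# Large mesh chunks bounce on a plane at a sampled skeletal-mesh position and feed ribbon trails via particle reads; smaller chunks skip collision.
# Time-of-day functions tint the layers, and front-end modifiers adjust smoke scale and camera offsets.
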NiagaraSystem /BRCosmetics/Animation/Game/MainPlayer/Emotes/TwiceBaked/FX/NS_Emote_TwiceBaked_Crush
dependencies {
 usedDataInterfaces [
  &NiagaraDataInterfacePlatformSet
  &NiagaraDataInterfaceArrayInt32
  &NiagaraDataInterfaceArrayFloat
  &NiagaraDataInterfaceCurve
  &NiagaraDataInterfaceArrayFloat3
  &NiagaraDataInterfaceSpriteRendererInfo
  &NiagaraDataInterfaceParticleRead
  &NiagaraDataInterfaceSkeletalMesh
 ]
 usedDynamicInputs [&SimulationPosition]
 usedModules [&SystemState, &SampleSkeletalMesh]
}
emitters {
 CrumbleDust {
  emitterUpdateScript [
   &EmitterState()
   &SpawnRate(
    Spawn Probability=&Multiply_Float(
     A=0.66
     B=&FloatFromCurve(
      FloatCurve=DataInterface() {propertyValues {Curve {keys [{tangentMode=RCTM_User, time=0, value=0}, {tangentMode=RCTM_User, time=1, value=1}]}}}
      CurveIndex=$System.NormalizedLoopAge
     )
    )
    SpawnRate=33
    Use Spawn Probability=true
   )
  ]
  particleSpawnScript [
   &InitializeParticle(
    Color Maximum=<rgba 0.1979 0.1979 0.1979 1>
    Color Minimum=<rgba 0.2135 0.2135 0.2135 1>
    Color Mode=::NewEnumerator2
    Lifetime Max=0.75
    Lifetime Min=0.5
    Lifetime Mode=::NewEnumerator1
    Sprite Rotation Mode=::NewEnumerator2
    Sprite Size Mode=::NewEnumerator1
    Uniform Sprite Size Max=3
    Uniform Sprite Size Min=1.5
    UsePositionOffset=false
    Write Material Random=true
   )
   &ShapeLocation(Box Size=<15 12 7>, MaskSpawnGroup=false, Override Seed=false, Shape Primitive=::NewEnumerator2, Surface Only=false, UseOwnerScale=true)
   &CameraOffset(Camera Offset Amount=5)
   &AddVelocity(
    Module.Use Falloff Distance=false
    Origin Offset Coordinate Space=::World
    Override Seed=false
    Velocity Mode=::NewEnumerator1
    Velocity Speed=&Multiply_Float(
     A=50
     B=&FloatFromCurve(
      FloatCurve=DataInterface() {propertyValues {Curve {keys [
          {arriveTangent=4.1785, interpMode=RCIM_Cubic, leaveTangent=4.1785, tangentMode=RCTM_User, time=0, value=0}
          {arriveTangent=1.1179, interpMode=RCIM_Cubic, leaveTangent=1.1179, tangentMode=RCTM_User, time=0.2042, value=0.6228}
          {arriveTangent=0.0224, interpMode=RCIM_Cubic, leaveTangent=0.0224, tangentMode=RCTM_User, time=1, value=1}
         ]}}}
      CurveIndex=$System.NormalizedLoopAge
     )
    )
   )
  ]
  particleUpdateScript [
   &ParticleState()
   &InheritSourceMovement(
    Applied Position Scale=&Multiply_VectorByFloat(
     Vector=<1 1 1>
     Float=&FloatFromCurve(FloatCurve=DataInterface() {propertyValues {Curve {keys [{interpMode=RCIM_Cubic, tangentMode=RCTM_User, time=0, value=1}, {interpMode=RCIM_Cubic, tangentMode=RCTM_User, time=1, value=0}]}}})
    )
    ApplyForce=false
    ApplyPosition=true
    ApplyPositionFalloff=false
   )
   &ScaleColor(
    Scale RGB=&MakeVectorFromLinearColorRGB(Color=&FNTimeOfDayColor(
      Night Color=<rgba 0.3875 0.5055 1 1>
      Morning Color=<rgba 1 0.6676 0.846 1>
      Day Color=<rgba 1 1 1 1>
      Evening Color=<rgba 1 0.5614 0.5913 1>
     ))
    ScaleA=true
    ScaleRGB=true
   )
   &SubUVAnimation(
    "Sprite Renderer "=DataInterface() {propertyValues {SpriteRenderer=&self:CrumbleDust_6.NiagaraSpriteRendererProperties_0}}
    UseEndFrame=false
    UseStartFrame=false
   )
   &GravityForce(Gravity=&RandomRangeVector(Minimum=<0 0 -600>, Maximum=<0 0 -980>))
   &SolveForcesAndVelocity(Clamp Velocity=false, Limit Acceleration=false)
   &FN_Sprite_Attributes(
    Color=<rgba 0 0 0 1>
    Create Color Attribute=true
    Create SpriteSize Attribute=true
    SpriteSize=&Multiply_Vector2D_ByFloat(Float=1.2, Vector2D=$Particles.SpriteSize)
    Write to Module Namespace=true
   )
   &KillParticlesInVolume(Box Size:Vector3f=<6 6 6>, Kill Shape=::NewEnumerator1)
  ]
  rendererClasses [&NiagaraSpriteRendererProperties]
  renderers [
   {rendererClass=&NiagaraSpriteRendererProperties}
   {rendererClass=&NiagaraSpriteRendererProperties, rendererIndex=1}
  ]
 }
 Directional_Smoke {
  emitterUpdateScript [
   &EmitterState(Life Cycle Mode=::NewEnumerator1, Loop Delay=$User.Delay, Loop Duration=1.406, Scalability Mode=::NewEnumerator1, UseLoopDelay=true)
   &SpawnBurst_Instantaneous(Spawn Count=2, Spawn Time=0.03, Use Spawn Probability=false)
  ]
  particleSpawnScript [
   &PartitionParticles(Particles Per Partition=1, Partition Mode:ENiagara_PartitionMode=::NewEnumerator1)
   &InitializeParticle(
    Color=<rgba 0.1979 0.1979 0.1979 0.5>
    Lifetime Max=0.5
    Lifetime Min=0.4
    Lifetime Mode=::NewEnumerator1
    Sprite Size Max=<40 75>
    Sprite Size Min=<30 55>
    Sprite Size Mode=::NewEnumerator2
    UsePositionOffset=false
    Write Material Random=true
   ) {module=InitializeParticle001}
   /SetVariables(Particles.Random=&RandomRangeFloat(Minimum=0, Maximum=1)) {isSetParameters=true, module=SetVariables_#1}
   &AddVelocity(
    Velocity=&SelectVectorFromArray(
     Direct Array Index=$Particles.PartitionParticles.Partition
     Vector Selection Array=DataInterface() {propertyValues {FloatData [<1 0 0>, <-1 0 0>]}}
    )
    Velocity Speed Scale=150
   )
   &CameraOffset(Camera Offset Amount=10)
   &DynamicMaterialParameters(
    Index 0 Param 2=&UniformAOrBFloat(A=0)
    Module.Param0WriteEnabled=true
    Module.Param1WriteEnabled=true
    Module.Param2WriteEnabled=true
    Module.Param3WriteEnabled=true
   ) {module=DynamicMaterialParameters001}
   /SetVariables(Particles.FlippableUVScale:Vector2f=&Lerp_Vector2(A=<-1 1>, B=<1 1>, Alpha=&UniformAOrBFloat(A=0))) {isSetParameters=true}
  ]
  particleUpdateScript [
   &ParticleState()
   &ScaleSpriteSize(
    Scale Sprite Size Mode=::NewEnumerator0
    Uniform Scale Factor=&FN_FrontEndFloatModifier(Gameplay=1, Front End=$User.FrontendScale)
   )
   &Drag(Drag=8, Use Linear Drag=true, Use Rotational Drag=false)
   &SolveForcesAndVelocity(Clamp Velocity=false, Limit Acceleration=false)
   &DynamicMaterialParameters(
    Index 0 Param 1=&FloatFromCurve(FloatCurve=DataInterface() {propertyValues {Curve {keys [
         {arriveTangent=1.2045, interpMode=RCIM_Cubic, leaveTangent=1.2045, tangentMode=RCTM_User, time=0, value=0}
         {arriveTangent=0.3642, interpMode=RCIM_Cubic, leaveTangent=0.3642, tangentMode=RCTM_User, time=1, value=1}
        ]}}})
    Index 0 Param 2=&UniformAOrBFloat(
     A=&MakeCustomFloatFromBool(
      Bool=&SetBoolByFloatComparison(A=&MakeFloatFromInt(Int=$User.SwingSectionIndex), B=3, Comparison Type=::NewEnumerator4)
      True Float=1
      False Float=0
     )
     Distribution Weight=&MakeFloatFromInt(Int=&RandomRangeInt(Minimum=0, Maximum=1))
    )
    Module.Param0WriteEnabled=true
    Module.Param1WriteEnabled=true
    Module.Param2WriteEnabled=true
    Module.Param3WriteEnabled=true
   )
   &ScaleColor(
    Color Value To Scale=$Particles.Color
    Scale Alpha=&Multiply_Float(
     A=&FNTimeOfDayScalar(Night Scale=0.3, Morning Scale=0.3, Evening Scale=0.3)
     B=&FloatFromCurve(FloatCurve=DataInterface() {propertyValues {Curve {keys [{interpMode=RCIM_Cubic, tangentMode=RCTM_User, time=0, value=1}, {interpMode=RCIM_Cubic, tangentMode=RCTM_User, time=1, value=0}]}}})
    )
    Scale RGB=&VectorFromFloat(Value=1)
    ScaleA=true
    ScaleRGB=true
   )
   &FN_Sprite_Attributes(
    Color=&FNTimeOfDayColor(
     Night Color=<rgba 0.2657 0.3435 0.5938 0.5>
     Morning Color=<rgba 0.5938 0.3912 0.4967 0.5>
     Day Color=<rgba 0.5938 0.5418 0.5052 0.5>
     Evening Color=<rgba 0.5938 0.3099 0.3634 0.5>
    )
    Create Color Attribute=true
    Create SpriteSize Attribute=true
    SpriteSize=&Multiply_Vector2D_ByFloat(Float=0.95, Vector2D=$Particles.SpriteSize)
    Write to Module Namespace=true
   )
  ]
  rendererClasses [&NiagaraSpriteRendererProperties]
  renderers [
   {rendererClass=&NiagaraSpriteRendererProperties}
   {rendererClass=&NiagaraSpriteRendererProperties, rendererIndex=1}
  ]
 }
 LargeRockTrail {
  emitterUpdateScript [
   &EmitterState(Life Cycle Mode=::NewEnumerator1, Loop Behavior=::NewEnumerator1, Loop Delay=$User.Delay, Loop Duration=1.25, UseLoopDelay=true)
   &SpawnParticlesFromOtherEmitter(
    Attribute Reader=DataInterface() {propertyValues {EmitterBinding {bindingMode=Other, emitterName=RockPieces}}}
    Spawn Rate=$Engine.InverseDeltaTime
    Spawning Enabled:NiagaraBool=&SetBoolByFloatComparison(A=0.6, B=$Emitter.LoopedAge)
    Use Maximum Age=false
    Use Minimum Age=false
   )
  ]
  particleSpawnScript [
   &InitializeParticle(
    Color=<rgba 0.125 0.1113 0.0914 1>
    Lifetime=0.4
    Ribbon Width=7
    Ribbon Width Mode=::NewEnumerator1
    UsePositionOffset=false
    Write Material Random=true
   )
   &SampleParticlesFromOtherEmitter(Emitter Sampling Mode:ENiagara_EmitterLocSamplingMode=::NewEnumerator0, Particle ID Sampling=::NewEnumerator1)
  ]
  particleUpdateScript [
   &GravityForce(Gravity=<0 0 -150>)
   &SolveForcesAndVelocity(Clamp Velocity=false, Limit Acceleration=false)
   &ParticleState()
   &ScaleColor(
    Scale Alpha=&FloatFromCurve(
     FloatCurve=DataInterface() {propertyValues {Curve {keys [{interpMode=RCIM_Cubic, tangentMode=RCTM_User, time=0.25, value=0.5}, {interpMode=RCIM_Cubic, tangentMode=RCTM_User, time=0.6, value=0}]}}}
     CurveIndex=$Emitter.LoopedAge
     Scale Curve=&FloatFromCurve(FloatCurve=DataInterface() {propertyValues {Curve {keys [{interpMode=RCIM_Cubic, tangentMode=RCTM_User, time=0, value=1}, {interpMode=RCIM_Cubic, tangentMode=RCTM_User, time=1, value=0}]}}})
    )
    Scale RGB=&MakeVectorFromLinearColorRGB(Color=&FNTimeOfDayColor(
      Night Color=<rgba 0.3875 0.5055 1 1>
      Morning Color=<rgba 1 0.6676 0.846 1>
      Day Color=<rgba 1 1 1 1>
      Evening Color=<rgba 1 0.5614 0.5913 1>
     ))
    ScaleA=true
    ScaleRGB=true
   )
   &ScaleRibbonWidth(Ribbon Width Scale=&FloatFromCurve(FloatCurve=DataInterface() {propertyValues {Curve {keys [{interpMode=RCIM_Cubic, tangentMode=RCTM_User, time=0, value=1}, {interpMode=RCIM_Cubic, tangentMode=RCTM_User, time=1, value=0.5}]}}}))
  ]
  rendererClasses [&NiagaraRibbonRendererProperties]
  renderers [{rendererClass=&NiagaraRibbonRendererProperties}]
 }
 RockPieces {
  emitterSpawnScript [
   /Game/Effects/Niagara/MasterMaterials/ToonShader/Template/NM_ToonShaderParameters(HighLight_Size=0.1636)
   /Game/Effects/Niagara/MasterMaterials/ToonShader/Template/NM_OutlineParameters(OutlineWidth:NiagaraFloat=&FN_PickFloatDependOnPlatform(TargetPlatformActive=1, TargetPlatformNoActive=0.4))
   /SetVariables(Emitter.Switch Outline Thickness=0.85) {isSetParameters=true, module=SetVariables_#2}
  ]
  emitterUpdateScript [
   &EmitterState(
    Life Cycle Mode=::NewEnumerator1
    Loop Behavior=::NewEnumerator1
    Loop Delay=$User.Delay
    Loop Duration Mode=::NewEnumerator1
    UseLoopDelay=true
   )
   &SpawnBurst_Instantaneous(Spawn Count=6, Use Spawn Probability=false, UseLoopCountLimit=false)
  ]
  particleSpawnScript [
   &PartitionParticles(
    Particles Per Partition:NiagaraInt32=&SelectIntFromArray(Direct Array Index=$Engine.QualityLevel, Int Selection Array=DataInterface() {propertyValues {IntData [2,2,3,3,3]}})
    Partition Mode:ENiagara_PartitionMode=::NewEnumerator1
   )
   &InitializeParticle(
    Color Maximum=<rgba 0.1667 0.1667 0.1667 1>
    Color Minimum=<rgba 0.5 0.5 0.5 1>
    Color Mode=::NewEnumerator2
    Lifetime Min=1.5
    Lifetime Mode=::NewEnumerator1
    Mass Mode=::NewEnumerator1
    Mesh Scale Mode=::NewEnumerator3
    Mesh Uniform Scale=&SelectFloatFromArray(
     Array Sampling Mode=::NewEnumerator0
     Direct Array Index=?
     Float Selection Array=DataInterface() {propertyValues {FloatData [0.7,0.8,0.9]}}
     Random Seed=0
    )
    UsePositionOffset=false
    Write Material Random=true
   )
   /SetVariables(Particles.MeshRandom:NiagaraInt32=&RandomRangeInt(Minimum=0, Maximum=2)) {isSetParameters=true, module=SetVariables_#3}
   &InitialMeshOrientation(Use Rotation Vector=false)
   &ShapeLocation(Box Size=<10 12 5>, MaskSpawnGroup=false, Override Seed=false, Shape Primitive=::NewEnumerator2, Surface Only=true, UseOwnerScale=true)
   &CameraOffset(Camera Offset Amount=&FN_FrontEndFloatModifier(Gameplay=-2, Front End=-5))
   &DynamicMaterialParameters(Module.Index0Param0Write=true, Module.Index0Param1Write=true, Module.Index0Param2Write=true, Module.Index0Param3Write=true)
   &AddVelocity(
    Cone Angle=75
    Cone Axis=&MaskVectorBySpawnGroup(Masked Vector=<1 0 0>, Passthrough Vector=<-1 0 0>, Spawn Group Mask=$Particles.PartitionParticles.Partition)
    Inner Cone Angle=65
    Module.UseConeFalloff=false
    Module.UseVelDistribution=false
    Override Seed=false
    Velocity Mode=::NewEnumerator2
    Velocity Speed=&RandomRangeFloat(Minimum=75, Maximum=150)
   )
  ]
  particleUpdateScript [
   &ParticleState()
   &InheritSourceMovement(
    Applied Position Scale=&Multiply_VectorByFloat(
     Vector=<1 1 1>
     Float=&FloatFromCurve(FloatCurve=DataInterface() {propertyValues {Curve {keys [{interpMode=RCIM_Cubic, tangentMode=RCTM_User, time=0, value=1}, {interpMode=RCIM_Cubic, tangentMode=RCTM_User, time=1, value=0}]}}})
    )
    ApplyForce=false
    ApplyPosition=true
    ApplyPositionFalloff=false
   )
   &ScaleColor(
    Scale RGB=&MakeVectorFromLinearColorRGB(Color=&FNTimeOfDayColor(
      Night Color=<rgba 0.3875 0.5055 1 1>
      Morning Color=<rgba 1 0.6676 0.846 1>
      Day Color=<rgba 1 1 1 1>
      Evening Color=<rgba 1 0.5614 0.5913 1>
     ))
    ScaleA=true
    ScaleRGB=true
   )
   &GravityForce()
   &Drag(Rotational Drag=2, Use Linear Drag=true, Use Rotational Drag=true)
   &MeshRotationForce(Rotation=<35 0 0>)
   &ScaleMeshSize(Scale Factor=&VectorFromFloat(Value=&FloatFromCurve(
      FloatCurve=DataInterface() {propertyValues {Curve {keys [{interpMode=RCIM_Cubic, tangentMode=RCTM_User, time=0.5, value=1}, {arriveTangent=-3.8399, tangentMode=RCTM_User, time=1, value=0}]}}}
      Scale Curve=&RandomRangeFloat(Minimum=0.3, Maximum=0.7)
     )))
   &UpdateMeshOrientation(
    Rotation Rate=&SelectFloatFromArray(
     Array Sampling Mode=::NewEnumerator0
     Direct Array Index=?
     Float Selection Array=DataInterface() {propertyValues {FloatData [0.5,0.7,-0.5,-0.25]}}
     Random Seed=0
    )
    Rotation Vector=&RandomVector(Vector Scale=1)
   )
   &Collision(
    Age Colliding Particles=false
    Analytical Collision Plane Position 1:Vector3f=$System.SampleSkeletalMesh.SampledPosition
    Analytical Collision Plane Space:ENiagaraCoordinateSpace=::World
    CPU Collision Type=::NewEnumerator3
    Collision Enabled=&SetBoolByFloatComparison(A=$Particles.Age, B=0.1)
    Friction=1.2
    KillOnCollision=false
    Module.Control Rest State Via Surface Normal=false
    Module.EnableMaxCollisionCount=false
    Module.Randomize Collision Normal Vector=false
    Restitution=&RandomRangeFloat(Minimum=0.25, Maximum=0.35)
   )
   &SolveForcesAndVelocity(Clamp Velocity=false, Limit Acceleration=false)
  ]
  rendererClasses [&NiagaraMeshRendererProperties]
  renderers [
   {rendererClass=&NiagaraMeshRendererProperties, rendererIndex=0}
   {rendererClass=&NiagaraMeshRendererProperties, rendererIndex=1}
   {rendererClass=&NiagaraMeshRendererProperties, rendererIndex=2}
  ]
 }
 RockPiecesSmall {
  emitterSpawnScript [
   &NM_ToonShaderParameters(HighLight_Size=0.1636)
   &NM_OutlineParameters(OutlineWidth:NiagaraFloat=&FN_PickFloatDependOnPlatform(TargetPlatformActive=1, TargetPlatformNoActive=0.4))
   /SetVariables(Emitter.Switch Outline Thickness=0.85) {isSetParameters=true, module=SetVariables_#2}
  ]
  emitterUpdateScript [
   &EmitterState(
    Life Cycle Mode=::NewEnumerator1
    Loop Behavior=::NewEnumerator1
    Loop Delay=$User.Delay
    Loop Duration Mode=::NewEnumerator1
    UseLoopDelay=true
   )
   &SpawnBurst_Instantaneous(Spawn Count=12, Use Spawn Probability=false, UseLoopCountLimit=false)
  ]
  particleSpawnScript [
   &PartitionParticles(
    Particles Per Partition:NiagaraInt32=&SelectIntFromArray(Direct Array Index=$Engine.QualityLevel, Int Selection Array=DataInterface() {propertyValues {IntData [4,4,6,6,6]}})
    Partition Mode:ENiagara_PartitionMode=::NewEnumerator1
   )
   &InitializeParticle(
    Color Maximum=<rgba 0.1667 0.1667 0.1667 1>
    Color Minimum=<rgba 0.5 0.5 0.5 1>
    Color Mode=::NewEnumerator2
    Mass Mode=::NewEnumerator1
    Mesh Scale Mode=::NewEnumerator3
    Mesh Uniform Scale=&SelectFloatFromArray(
     Array Sampling Mode=::NewEnumerator0
     Direct Array Index=?
     Float Selection Array=DataInterface() {propertyValues {FloatData [0.7,0.8,0.9]}}
     Random Seed=0
    )
    UsePositionOffset=false
    Write Material Random=true
   )
   /SetVariables(Particles.MeshRandom:NiagaraInt32=&RandomRangeInt(Minimum=0, Maximum=2)) {isSetParameters=true, module=SetVariables_#3}
   &InitialMeshOrientation(Use Rotation Vector=false)
   &ShapeLocation(Box Size=<10 12 5>, MaskSpawnGroup=false, Override Seed=false, Shape Primitive=::NewEnumerator2, Surface Only=true, UseOwnerScale=true)
   &CameraOffset(Camera Offset Amount=&FN_FrontEndFloatModifier(Gameplay=-2, Front End=-5))
   &DynamicMaterialParameters(Module.Index0Param0Write=true, Module.Index0Param1Write=true, Module.Index0Param2Write=true, Module.Index0Param3Write=true)
   &AddVelocity(
    Cone Angle=80
    Cone Axis=&MaskVectorBySpawnGroup(Masked Vector=<1 0 0>, Passthrough Vector=<-1 0 0>, Spawn Group Mask=$Particles.PartitionParticles.Partition)
    Inner Cone Angle=60
    Module.UseConeFalloff=false
    Module.UseVelDistribution=false
    Override Seed=false
    Velocity Mode=::NewEnumerator2
    Velocity Speed=&RandomRangeFloat(Minimum=100, Maximum=300)
   )
  ]
  particleUpdateScript [
   &ParticleState()
   &InheritSourceMovement(
    Applied Position Scale=&Multiply_VectorByFloat(
     Vector=<1 1 1>
     Float=&FloatFromCurve(FloatCurve=DataInterface() {propertyValues {Curve {keys [{interpMode=RCIM_Cubic, tangentMode=RCTM_User, time=0, value=1}, {interpMode=RCIM_Cubic, tangentMode=RCTM_User, time=1, value=0}]}}})
    )
    ApplyForce=false
    ApplyPosition=true
    ApplyPositionFalloff=false
   )
   &ScaleColor(
    Scale RGB=&MakeVectorFromLinearColorRGB(Color=&FNTimeOfDayColor(
      Night Color=<rgba 0.3875 0.5055 1 1>
      Morning Color=<rgba 1 0.6676 0.846 1>
      Day Color=<rgba 1 1 1 1>
      Evening Color=<rgba 1 0.5614 0.5913 1>
     ))
    ScaleA=true
    ScaleRGB=true
   )
   &GravityForce()
   &Drag(Rotational Drag=2, Use Linear Drag=true, Use Rotational Drag=true)
   &MeshRotationForce(Rotation=<35 0 0>)
   &ScaleMeshSize(Scale Factor=&VectorFromFloat(Value=&FloatFromCurve(
      FloatCurve=DataInterface() {propertyValues {Curve {keys [{interpMode=RCIM_Cubic, tangentMode=RCTM_User, time=0.5, value=1}, {arriveTangent=-3.8399, tangentMode=RCTM_User, time=1, value=0}]}}}
      Scale Curve=&RandomRangeFloat(Minimum=0.3, Maximum=0.7)
     )))
   &UpdateMeshOrientation(
    Rotation Rate=&SelectFloatFromArray(
     Array Sampling Mode=::NewEnumerator0
     Direct Array Index=?
     Float Selection Array=DataInterface() {propertyValues {FloatData [0.5,0.7,-0.5,-0.25]}}
     Random Seed=0
    )
    Rotation Vector=&RandomVector(Vector Scale=1)
   )
   &SolveForcesAndVelocity(Clamp Velocity=false, Limit Acceleration=false)
  ]
  rendererClasses [&NiagaraMeshRendererProperties]
  renderers [
   {rendererClass=&NiagaraMeshRendererProperties, rendererIndex=0}
   {rendererClass=&NiagaraMeshRendererProperties, rendererIndex=1}
   {rendererClass=&NiagaraMeshRendererProperties, rendererIndex=2}
  ]
 }
 SmallDust {
  emitterUpdateScript [
   &EmitterState(Life Cycle Mode=::NewEnumerator1, Loop Behavior=::NewEnumerator1, Loop Delay=$User.Delay, Loop Duration=0.4, UseLoopDelay=true)
   &SpawnBurst_Instantaneous(Spawn Count=10, Use Spawn Probability=false, UseLoopCountLimit=false)
  ]
  particleSpawnScript [
   &PartitionParticles(
    Particles Per Partition:NiagaraInt32=&SelectIntFromArray(Direct Array Index=$Engine.QualityLevel, Int Selection Array=DataInterface() {propertyValues {IntData [4,4,5,5,5]}})
    Partition Mode:ENiagara_PartitionMode=::NewEnumerator1
   )
   &InitializeParticle(
    Color=&FNTimeOfDayColor(
     Night Color=<rgba 0.0427 0.0582 0.105 1>
     Morning Color=<rgba 0.105 0.059 0.0728 1>
     Day Color=<rgba 0.105 0.105 0.105 1>
     Evening Color=<rgba 0.105 0.069 0.0617 1>
    )
    Lifetime=0.7
    Sprite Rotation Mode=::NewEnumerator2
    Sprite Size Mode=::NewEnumerator1
    Uniform Sprite Size Max=80
    Uniform Sprite Size Min=70
    UsePositionOffset=false
    Write Material Random=true
   )
   &ShapeLocation(MaskSpawnGroup=false, Override Seed=false, Sphere Radius=2, UseOwnerScale=true)
   &AddVelocity(
    Cone Axis=&MaskVectorBySpawnGroup(Masked Vector=<1 0 0>, Passthrough Vector=<-1 0 0>, Spawn Group Mask=$Particles.PartitionParticles.Partition)
    Inner Cone Angle=25
    Module.UseConeFalloff=false
    Module.UseVelDistribution=false
    Override Seed=false
    Velocity Mode=::NewEnumerator2
    Velocity Speed=&RandomRangeFloat(Minimum=75, Maximum=100)
   )
   &CameraOffset(Camera Offset Amount=&FN_FrontEndFloatModifier(Gameplay=-2, Front End=-5))
  ]
  particleUpdateScript [
   &ParticleState()
   &InheritSourceMovement(
    Applied Position Scale=&Multiply_VectorByFloat(
     Vector=<1 1 1>
     Float=&FloatFromCurve(FloatCurve=DataInterface() {propertyValues {Curve {keys [{interpMode=RCIM_Cubic, tangentMode=RCTM_User, time=0, value=1}, {interpMode=RCIM_Cubic, tangentMode=RCTM_User, time=1, value=0}]}}})
    )
    ApplyForce=false
    ApplyPosition=true
    ApplyPositionFalloff=false
   )
   &GravityForce(Gravity=&RandomRangeVector(Minimum=<0 0 -350>, Maximum=<0 0 -550>))
   &Drag(Use Linear Drag=true, Use Rotational Drag=false)
   &ScaleColor(
    Scale Alpha=&FloatFromCurve(FloatCurve=DataInterface() {propertyValues {Curve {keys [<at 0 1>,<at 1 0.5>]}}})
    ScaleA=true
    ScaleRGB=true
   )
   &SolveForcesAndVelocity(Clamp Velocity=false, Limit Acceleration=false)
   &ScaleSpriteSize(Scale Factor=&Vector2DFromFloat(Value=&FloatFromCurve(FloatCurve=DataInterface() {propertyValues {Curve {keys [
          {arriveTangent=1.4242, interpMode=RCIM_Cubic, leaveTangent=1.4242, tangentMode=RCTM_User, time=0, value=0.2}
          {arriveTangent=-0.2, interpMode=RCIM_Cubic, leaveTangent=-0.2, time=0.9, value=1}
          <at 1 0 RCIM_Cubic>
         ]}}})))
  ]
  rendererClasses [&NiagaraSpriteRendererProperties]
  renderers [{rendererClass=&NiagaraSpriteRendererProperties}]
 }
 Smoke {
  emitterUpdateScript [
   &EmitterState(Life Cycle Mode=::NewEnumerator1, Loop Delay=$User.Delay, Loop Duration=1, Scalability Mode=::NewEnumerator1, UseLoopDelay=true)
   &SpawnBurst_Instantaneous(Spawn Count=1, Use Spawn Probability=false)
  ]
  particleSpawnScript [
   /SetVariables(Particles.Random=&RandomRangeFloat(Minimum=0, Maximum=1)) {isSetParameters=true, module=SetVariables_#1}
   &InitializeParticle(
    Color=<rgba 0.1979 0.1979 0.1979 0.5>
    Lifetime=&Lerp_Float(A=0.3, B=0.35, Alpha=$Particles.Random)
    Mass Mode=::NewEnumerator1
    Sprite Rotation Angle=&Lerp_Float(A=-45, B=45, Alpha=$Particles.Random)
    Sprite Rotation Mode=::NewEnumerator0
    Sprite Size=&Vector2DFromFloat(Value=&Lerp_Float(A=62, B=70, Alpha=$Particles.Random))
    Sprite Size Mode=::NewEnumerator4
    UsePositionOffset=false
    Write Material Random=true
   ) {module=InitializeParticle001}
   &DynamicMaterialParameters(
    Index 0 Param 2=&UniformAOrBFloat(A=0)
    Module.Param0WriteEnabled=true
    Module.Param1WriteEnabled=true
    Module.Param2WriteEnabled=true
    Module.Param3WriteEnabled=true
   ) {module=DynamicMaterialParameters001}
   &CameraOffset(Camera Offset Amount=10)
  ]
  particleUpdateScript [
   &ParticleState()
   &SolveForcesAndVelocity(Clamp Velocity=false, Limit Acceleration=false)
   &DynamicMaterialParameters(
    Index 0 Param 1=&FloatFromCurve(FloatCurve=DataInterface() {propertyValues {Curve {keys [<at 0 0>,<at 1 1>]}}})
    Index 0 Param 2=1
    Index 0 Param 3=0.9
    Module.Param0WriteEnabled=true
    Module.Param1WriteEnabled=true
    Module.Param2WriteEnabled=true
    Module.Param3WriteEnabled=true
   )
   &ScaleColor(
    Scale Alpha=&FloatFromCurve(FloatCurve=DataInterface() {propertyValues {Curve {keys [{interpMode=RCIM_Cubic, tangentMode=RCTM_User, time=0.5, value=1}, {interpMode=RCIM_Cubic, tangentMode=RCTM_User, time=1, value=0}]}}})
    Scale RGB=&MakeVectorFromLinearColorRGB(Color=$Particles.Initial.Color)
    ScaleA=true
    ScaleRGB=true
   )
   &FN_Sprite_Attributes(
    Color=&FNTimeOfDayColor(
     Night Color=<rgba 0.2657 0.3435 0.5938 0.5>
     Morning Color=<rgba 0.5938 0.3912 0.4967 0.5>
     Day Color=<rgba 0.5938 0.5418 0.5052 0.5>
     Evening Color=<rgba 0.5938 0.3099 0.3634 0.5>
    )
    Create Color Attribute=true
    Create SpriteSize Attribute=true
    SpriteSize=&Multiply_Vector2D_ByFloat(Float=0.95, Vector2D=$Particles.SpriteSize)
    Write to Module Namespace=true
   )
  ]
  rendererClasses [&NiagaraSpriteRendererProperties]
  renderers [
   {rendererClass=&NiagaraSpriteRendererProperties}
   {rendererClass=&NiagaraSpriteRendererProperties, rendererIndex=1}
  ]
 }
}
summary {
 systemName=NS_Emote_TwiceBaked_Crush
 emitters [
  {emitterName=RockPieces, rendererClasses [&NiagaraMeshRendererProperties]}
  {emitterName=Smoke, rendererClasses [&NiagaraSpriteRendererProperties]}
  {emitterName=Directional_Smoke, rendererClasses [&NiagaraSpriteRendererProperties]}
  {emitterName=SmallDust, rendererClasses [&NiagaraSpriteRendererProperties]}
  {emitterName=CrumbleDust, rendererClasses [&NiagaraSpriteRendererProperties]}
  {emitterName=LargeRockTrail, rendererClasses [&NiagaraRibbonRendererProperties]}
  {emitterName=RockPiecesSmall, rendererClasses [&NiagaraMeshRendererProperties]}
 ]
 userVariables [
  User.bIsFrontEnd {type=NiagaraBool}
  User.bIsFrontEndPreview {type=NiagaraBool}
  User.Delay {type=NiagaraFloat, defaultValue=1.9, description=""}
  User.FrontendScale {type=NiagaraFloat, defaultValue=0, description=""}
  User.SwingSectionIndex {type=NiagaraInt32}
 ]
}
